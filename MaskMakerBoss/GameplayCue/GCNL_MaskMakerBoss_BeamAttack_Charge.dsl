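; Looping gameplay cue for the MaskMaker boss beam attack charge-up, driving its effects toward the boss's current target.
; On loop start it caches the boss pawn, weakpoint actors and looping FX and enables tick, or removes itself if the cast fails.
; Each tick writes the target's chest socket location back to the pawn and feeds beam end, attack flag and root decal location to the FX.
; Tick stops when the pawn stops updating that location or on removal.
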(event EventOnLoopingStart(EitherVFX-preferthisversion)
  (params (MyTarget :type Actor%20Object%20Reference) (Parameters :type Gameplay%20Cue%20Parameters%20Structure%20%28by%20ref%29) (ParticleComponents :type Array%20of%20FXSystem%20Component%20Object%20References) (AudioComponents :type Array%20of%20Audio%20Component%20Object%20References))
  (Variables|Default|SetLoopingFX ParticleComponents)
  (bind _asbp_player_pawn_cherry_smoke_mask_maker_boss (Utilities|Casting|CastToBP_PlayerPawn_CherrySmoke_MaskMakerBoss MyTarget)
    (:then
      (bind _output_get (Variables|Default|SetMaskMakerPawn _asbp_player_pawn_cherry_smoke_mask_maker_boss))
      (bind _outactors (Actor|GetAllActorsOfClass _output_get "/MaskMakerBoss/Weakpoint/BGA_WeakpointActor_MaskMaker.BGA_WeakpointActor_MaskMaker_C"))
      (Variables|Default|SetMaskWeakpoints _outactors)
      (Actor|Tick|SetActorTickEnabled true)
      (|CallParentFunction MyTarget Parameters ParticleComponents AudioComponents))
    (:CastFailed
      (GameplayCue|RemoveGameplayCue self (CallFunction|GetGameplayCueTag)))))

(event EventTick
  (params (DeltaSeconds :type Float%20%28single-precision%29))
  (bind _maskmakerpawn (Variables|Default|GetMaskMakerPawn))
  (bind _currentbosstarget (Variables|/MaskMakerBoss/Pawn/BP_PlayerPawn_CherrySmoke_MaskMakerBoss.BP_PlayerPawn_CherrySmoke_MaskMakerBoss_C|GetCurrentBossTarget _maskmakerpawn))
  (bind _mesh (Variables|/Script/Engine.Character|GetMesh _currentbosstarget))
  (bind _loopingfx (Variables|Default|GetLoopingFX))
  (Utilities|IsValid _currentbosstarget
    (:Is%20Valid
      (if (Variables|/MaskMakerBoss/Pawn/BP_PlayerPawn_CherrySmoke_MaskMakerBoss.BP_PlayerPawn_CherrySmoke_MaskMakerBoss_C|GetUpdateTargetLastLoc (Variables|Default|GetMaskMakerPawn))
        (bind _output_get (Variables|/MaskMakerBoss/Pawn/BP_PlayerPawn_CherrySmoke_MaskMakerBoss.BP_PlayerPawn_CherrySmoke_MaskMakerBoss_C|SetTargetLastLocation (Transformation|GetSocketLocation _mesh "FX_Chest") (Variables|Default|GetMaskMakerPawn)))
        (Effects|Components|ParticleSystem|SetVectorParameter _loopingfx "Stop Location" _output_get)
        (Effects|Components|ParticleSystem|SetBoolParameter _loopingfx "IsAboutToAttack" true)
        (Effects|Components|ParticleSystem|SetVectorParameter _loopingfx "DecalLocation" (Transformation|GetSocketLocation _mesh "root"))
        (else
          (Actor|Tick|SetActorTickEnabled false))))))

(event EventOnRemoval(EitherVFX-preferthisversion)
  (params (MyTarget :type Actor%20Object%20Reference) (Parameters :type Gameplay%20Cue%20Parameters%20Structure%20%28by%20ref%29) (ParticleComponents :type Array%20of%20FXSystem%20Component%20Object%20References) (AudioComponents :type Array%20of%20Audio%20Component%20Object%20References) (BurstCameraShakeInstance :type Camera%20Shake%20Base%20Object%20Reference) (BurstDecalInstance :type Decal%20Actor%20Object%20Reference))
  (Actor|Tick|SetActorTickEnabled false)
  (|CallParentFunction MyTarget Parameters ParticleComponents AudioComponents BurstCameraShakeInstance BurstDecalInstance))

(fn ConstructionScript)
