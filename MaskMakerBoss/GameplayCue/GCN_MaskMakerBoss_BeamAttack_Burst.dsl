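; GameplayCue burst for the MaskMaker boss's beam attack that sets where the beam VFX ends.
; Two multi-line traces run from hand_r toward the boss's last target location, ignoring boss and weakpoints.
; Shield bubbles, non-damageable actors or exceeding the wall pass-through limit mark a blocking point, with distance checks favoring the nearest.
; That point sets particle stop location and decal; otherwise, first-trace impact or target location is used, without decal.
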(fn OnBurst(EitherVFX-preferthisversion)
  (params (MyTarget :type Actor%20Object%20Reference) (Parameters :type Gameplay%20Cue%20Parameters%20Structure%20%28by%20ref%29) (ParticleComponents :type Array%20of%20FXSystem%20Component%20Object%20References) (AudioComponents :type Array%20of%20Audio%20Component%20Object%20References) (BurstCameraShakeInstance :type Camera%20Shake%20Base%20Object%20Reference) (BurstDecalInstance :type Decal%20Actor%20Object%20Reference))
  (bind _beamattackmaxrange_hf (LocalVariables|GetBeamAttackMaxRange_Hf))
  (bind _returnvalue (ScalableFloat|GetValueatLevel _beamattackmaxrange_hf))
  (bind _maskmakerpawn (LocalVariables|GetMaskMakerPawn))
  (bind _targetlastlocation (Variables|/MaskMakerBoss/Pawn/BP_PlayerPawn_CherrySmoke_MaskMakerBoss.BP_PlayerPawn_CherrySmoke_MaskMakerBoss_C|GetTargetLastLocation _maskmakerpawn))
  (bind _maskmakerpawn_2 (LocalVariables|GetMaskMakerPawn))
  (bind _mesh (Variables|/Script/Engine.Character|GetMesh _maskmakerpawn_2))
  (bind _returnvalue_2 (Transformation|GetSocketLocation _mesh "hand_r"))
  (bind _returnvalue_3 (Math|Vector|GetUnitDirection(Vector) _returnvalue_2 _targetlastlocation))
  (bind _maskweakpoints (LocalVariables|GetMaskWeakpoints))
  (bind _output (Utilities|Array|Get(acopy) _maskweakpoints 3))
  (bind _output_2 (Utilities|Array|Get(acopy) _maskweakpoints))
  (bind _output_3 (Utilities|Array|Get(acopy) _maskweakpoints 1))
  (bind _output_4 (Utilities|Array|Get(acopy) _maskweakpoints 2))
  (bind _maskmakerpawn_3 (LocalVariables|GetMaskMakerPawn))
  (bind _array (Utilities|Array|MakeArray _maskmakerpawn_3 _output_2 _output_3 _output_4 _output))
  (bind _returnvalue_4 (* _returnvalue_3 _returnvalue))
  (bind _returnvalue_5 (+ _returnvalue_2 _returnvalue_4))
  (bind _maskmakerpawn_4 (LocalVariables|GetMaskMakerPawn))
  (bind _targetlastlocation_2 (Variables|/MaskMakerBoss/Pawn/BP_PlayerPawn_CherrySmoke_MaskMakerBoss.BP_PlayerPawn_CherrySmoke_MaskMakerBoss_C|GetTargetLastLocation _maskmakerpawn_4))
  (bind _hitindestructibleobject (LocalVariables|GetHitIndestructibleObject))
  (bind _maskmakerpawn_5 (LocalVariables|GetMaskMakerPawn))
  (bind _particlecomponents (Variables|GetParticleComponents))
  (bind _actorshitinfo (LocalVariables|GetActorsHitInfo))
  (|CallParentFunction MyTarget Parameters ParticleComponents AudioComponents BurstCameraShakeInstance BurstDecalInstance)
  (bind _asbp_player_pawn_cherry_smoke_mask_maker_boss (Utilities|Casting|CastToBP_PlayerPawn_CherrySmoke_MaskMakerBoss MyTarget))
  (bind _output_get (LocalVariables|SetMaskMakerPawn _asbp_player_pawn_cherry_smoke_mask_maker_boss))
  (bind _outactors (Actor|GetAllActorsOfClass _output_get "/MaskMakerBoss/Weakpoint/BGA_WeakpointActor_MaskMaker.BGA_WeakpointActor_MaskMaker_C"))
  (LocalVariables|SetMaskWeakpoints _outactors)
  (Utilities|Array|Clear (LocalVariables|GetActorsHitInfo))
  (LocalVariables|SetHitIndestructibleObject false)
  (LocalVariables|SetActorsHit 0)
  (bind (_outhits _returnvalue_6) (Collision|MultiLineTraceByChannel _maskmakerpawn_5 _returnvalue_2 _returnvalue_5 "TraceTypeQuery10" false _array))
  (if _returnvalue_6
    (for _array_element _outhits
      (bind _output_get_2 (LocalVariables|SetBurstBeamEndLocation (.ImpactPoint (Collision|BreakHitResult _array_element))))
      (Development|PrintString 0 "BlockingHit" false false "(R=0.160676,G=1.000000,B=0.068925,A=1.000000)" 22.0)
      (Development|PrintString 0 (Utilities|String|ToString(Vector) _output_get_2) false false "(R=0.184133,G=1.000000,B=0.052952,A=1.000000)" 22.0))
    (Utilities|Array|AppendArray (LocalVariables|GetActorsHitInfo) _outhits)
    (else
      ;[0.7708 0.0385 0.04586 0.5] If the line tracer doesn't hit anything, then use the target last location from the boss pawn, otherwise the beam target will be (0,0,0)
      (LocalVariables|SetBurstBeamEndLocation _targetlastlocation_2)
      (Development|PrintString 0 "NoBlockingHit" false false "(R=1.000000,G=0.000000,B=0.081085,A=1.000000)" 22.0)
      (Development|PrintString 0 (Utilities|String|ToString(Vector) _targetlastlocation_2) false false "(R=1.000000,G=0.000000,B=0.081085,A=1.000000)" 22.0)))
  (bind _outhits_2 (Collision|MultiLineTraceByChannel _maskmakerpawn_5 _returnvalue_2 _returnvalue_5 "TraceTypeQuery9" false _array))
  (Utilities|Array|AppendArray _actorshitinfo _outhits_2)
  (for _array_element_2 _actorshitinfo
    (Utilities|IsValid (.HitActor (Collision|BreakHitResult _array_element_2))
      (:Is%20Valid
        (if (== (Utilities|GetClass (.HitActor (Collision|BreakHitResult _array_element_2))) "/ShieldBubble/Gameplay/ShieldBubbleJr/BGA_Athena_SilverBlazer_Mini.BGA_Athena_SilverBlazer_Mini_C")
          (if _hitindestructibleobject :!label _label_2
            (if (> (Math|Vector|Distance(Vector) (Transformation|GetActorLocation (LocalVariables|GetMaskMakerPawn)) (LocalVariables|GetIndestructibleObjectLoc)) (Math|Vector|Distance(Vector) (Transformation|GetActorLocation (LocalVariables|GetMaskMakerPawn)) (.ImpactPoint (Collision|BreakHitResult _array_element_2)))) :!label _label_3
              _)
            (else
              (LocalVariables|SetHitIndestructibleObject true :!label _label)))
          (LocalVariables|SetIndestructibleObjectLoc (.ImpactPoint (Collision|BreakHitResult _array_element_2)))
          (LocalVariables|SetIndestructibleObjectNormal (.ImpactNormal (Collision|BreakHitResult _array_element_2)))
          (elif (Variables|/Script/Engine.Actor|GetbCanBeDamaged (.HitActor (Collision|BreakHitResult _array_element_2)))
            (if _hitindestructibleobject
              _
              (elif (< (LocalVariables|GetActorsHit) (ScalableFloat|GetValueasInteger (LocalVariables|GetMaxWallTraceThrough_Hf)))
                (Math|Integer|IncrementInt (LocalVariables|GetActorsHit))
                (else
                  (goto _label))))
            (else
              (goto _label_2))))
        (goto _label_3))))
  (if (LocalVariables|GetHitIndestructibleObject)
    (Effects|Components|ParticleSystem|SetVectorParameter _particlecomponents "Stop Location" (LocalVariables|GetIndestructibleObjectLoc))
    (Effects|Components|ParticleSystem|SetVectorParameter _particlecomponents "DecalNormal" (LocalVariables|GetIndestructibleObjectNormal))
    (Effects|Components|ParticleSystem|SetBoolParameter _particlecomponents "SpawnDecal" true)
    (else
      (Effects|Components|ParticleSystem|SetVectorParameter _particlecomponents "Stop Location" (LocalVariables|GetBurstBeamEndLocation))
      (Effects|Components|ParticleSystem|SetBoolParameter _particlecomponents "SpawnDecal" false)))
  (return))
