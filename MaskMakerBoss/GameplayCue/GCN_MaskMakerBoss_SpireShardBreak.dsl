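; Burst gameplay cue for Mask Maker boss spire shard breaks that aligns its Niagara effect to the surface at the target.
; After the parent burst logic runs, a vertical line trace through the target, ignoring the target itself, repositions the first particle component at the hit and orients it to the surface normal.
; A Mask Maker slag prop tag check on the target sets the effect's IsMaskMakerShards parameter.
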(fn OnBurst(Niagara)
  (params (MyTarget :type Actor%20Object%20Reference) (Parameters :type Gameplay%20Cue%20Parameters%20Structure%20%28by%20ref%29) (ParticleComponents :type Array%20of%20Niagara%20Particle%20System%20Component%20Object%20References) (AudioComponents :type Array%20of%20Audio%20Component%20Object%20References) (BurstCameraShakeInstance :type Camera%20Shake%20Base%20Object%20Reference) (BurstDecalInstance :type Decal%20Actor%20Object%20Reference))
  (bind _output (Utilities|Array|Get(acopy) ParticleComponents))
  (|CallParentFunction MyTarget Parameters ParticleComponents AudioComponents BurstCameraShakeInstance BurstDecalInstance)
  (Utilities|Array|Add (Variables|Default|GetActors%20to%20Ignore) MyTarget)
  (bind _outhit (Collision|LineTraceByChannel 0 (+ (Transformation|GetActorLocation MyTarget) "0.000000,0.000000,200.000000") (+ (Transformation|GetActorLocation MyTarget) "0.000000,0.000000,-200.000000") "TraceTypeQuery10" false (Variables|Default|GetActors%20to%20Ignore)))
  (Transformation|SetWorldLocation _output (.Location (Collision|BreakHitResult _outhit)) false false)
  (Transformation|SetWorldRotation _output (Math|Rotator|MakeRotfromZ (.Normal (Collision|BreakHitResult _outhit))) false false)
  (bind _returnvalue (GameplayTags|HasMatchingGameplayTag(Message) MyTarget "(TagName=\"Building.Type.Prop.Slag.MM\")"))
  (Effects|Components|ParticleSystem|SetBoolParameter _output "IsMaskMakerShards" _returnvalue)
  (return))
