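; Gameplay cue that shows the telegraph VFX for the Mask Maker boss's teleport AOE attack.
; On burst it calls the parent, casts the target to the boss pawn and attaches laser charge-up and teleport AOE charge-up Niagara effects to the mesh's FX_Chest socket.
; The AOE effect receives the boss's location as SourcePosition and the cue's Location parameter as TargetPosition.
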(fn OnBurst(EitherVFX-preferthisversion)
  (params (MyTarget :type Actor%20Object%20Reference) (Parameters :type Gameplay%20Cue%20Parameters%20Structure%20%28by%20ref%29) (ParticleComponents :type Array%20of%20FXSystem%20Component%20Object%20References) (AudioComponents :type Array%20of%20Audio%20Component%20Object%20References) (BurstCameraShakeInstance :type Camera%20Shake%20Base%20Object%20Reference) (BurstDecalInstance :type Decal%20Actor%20Object%20Reference))
  (bind (_normalizedmagnitude _rawmagnitude _effectcontext _matchedtagname _originaltag _aggregatedsourcetags _aggregatedtargettags _location) (Ability|GameplayCue|BreakGameplayCueParameters Parameters))
  (|CallParentFunction MyTarget Parameters ParticleComponents AudioComponents BurstCameraShakeInstance BurstDecalInstance)
  (bind _asbp_player_pawn_cherry_smoke_mask_maker_boss (Utilities|Casting|CastToBP_PlayerPawn_CherrySmoke_MaskMakerBoss MyTarget))
  (Effects|SpawnFXSystemAttached "/MaskMakerBoss/VFX/NS_MaskMaker_LaserChargeup.NS_MaskMaker_LaserChargeup" (Variables|/Script/Engine.Character|GetMesh _asbp_player_pawn_cherry_smoke_mask_maker_boss) "FX_Chest")
  (bind _returnvalue (Effects|SpawnFXSystemAttached "/MaskMakerBoss/VFX/NS_MaskMaker_TeleportAOE_Chargeup.NS_MaskMaker_TeleportAOE_Chargeup" (Variables|/Script/Engine.Character|GetMesh _asbp_player_pawn_cherry_smoke_mask_maker_boss) "FX_Chest"))
  (Effects|Components|ParticleSystem|SetVectorParameter _returnvalue "SourcePosition" (Transformation|GetActorLocation MyTarget))
  (Effects|Components|ParticleSystem|SetVectorParameter _returnvalue "TargetPosition" _location)
  (return))
